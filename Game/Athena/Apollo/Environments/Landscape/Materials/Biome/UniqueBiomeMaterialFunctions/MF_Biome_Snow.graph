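# Landscape material function that defines the snow layer for the Apollo biome terrain, outputting Material attributes and a Height value.
# A tiled snow mask texture's blue channel drives both base color and roughness, while a separately tiled normal map parameter is partly flattened.
# Height lerps between two scalar parameters using the mask's green channel.
# A test color tint multiply and an extra coordinate node are left unconnected.
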MaterialFunction /Game/Athena/Apollo/Environments/Landscape/Materials/Biome/UniqueBiomeMaterialFunctions/MF_Biome_Snow
graph {nodes [
  n1 = TextureSample(UVs=LandscapeLayerCoords() {mappingScale=25}) {
   samplerSource=SSM_Wrap_WorldGroupSettings
   samplerType=SAMPLERTYPE_LinearColor
   texture=/Game/Athena/Environments/Landscape/Textures/T_Athena_Snow_01_M
  }
  n2 = TextureSampleParameter2D(UVs=LandscapeLayerCoords() {mappingScale=20}) {parameterName=Snow_Normal, samplerType=SAMPLERTYPE_Normal, texture=/Game/Athena/Environments/Landscape/Textures/T_Athena_Snow_01_N}
  n3 = Constant() {r=0.5}
  FunctionOutput(None=SetMaterialAttributes!(
    MaterialAttributes
    BaseColor=@n1.B
    Roughness=@n1.B
    Normal=MaterialFunctionCall(Normal=@n2.RGB, Flatness=@n3) {materialFunction=&FlattenNormal}.Result
   ) {attributeSetTypes [#1, #2, #3]}) {outputName=Material}
  FunctionOutput(None=LinearInterpolate(
    A=ScalarParameter() {defaultValue=0.6, parameterName=SnowHeightA}
    B=ScalarParameter() {defaultValue=1, parameterName=SnowHeightB}
    Alpha=@n1.G
   )) {outputName=Height, sortPriority=1}
  LandscapeLayerCoords() {mappingScale=5}
  Multiply(A=VectorParameter() {parameterName=SnowTestColor, defaultValue=<rgba 1 1 1 1>}.RGB, B=@n1.B)
 ]}
