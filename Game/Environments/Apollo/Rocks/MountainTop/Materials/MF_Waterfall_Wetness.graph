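# Material function that applies a wet-surface look to incoming material attributes, named for waterfall wetness on the Apollo mountain-top rocks.
# A scalar Mask input is scaled by a Mask Bias parameter and clamped to form the wetness blend alpha.
# That alpha lerps base color and roughness toward parameter-multiplied versions and specular toward an offset version, then writes all three back into the attributes.
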MaterialFunction /Game/Environments/Apollo/Rocks/MountainTop/Materials/MF_Waterfall_Wetness
graph {nodes [
  n1 = GetMaterialAttributes(None=FunctionInput() {inputType=FunctionInput_MaterialAttributes}) {attributeGetTypes [#1, #2, #3]}
  n2 = Multiply(A=ScalarParameter() {defaultValue=0.4, group=Wetness, parameterName=Wetness Darken}, B=@n1.BaseColor)
  n3 = Clamp(None=LinearInterpolate(
    A=Constant()
    B=ScalarParameter() {defaultValue=3, group=Wetness, parameterName=Mask Bias}
    Alpha=FunctionInput() {inputName=Mask, inputType=FunctionInput_Scalar}
   ))
  n4 = Multiply(A=ScalarParameter() {defaultValue=0.5, group=Wetness, parameterName=Wetness Roughness Bias}, B=@n1.Roughness)
  n5 = Add(A=ScalarParameter() {defaultValue=0.2, group=Wetness, parameterName=Specular Bias}, B=@n1.Specular)
  FunctionOutput(None=SetMaterialAttributes!(
    MaterialAttributes=@n1.MaterialAttributes
    BaseColor=LinearInterpolate(A=@n1.BaseColor, B=@n2, Alpha=@n3)
    Roughness=LinearInterpolate(A=@n1.Roughness, B=@n4, Alpha=@n3)
    Specular=LinearInterpolate(A=@n1.Specular, B=@n5, Alpha=@n3)
   ) {attributeSetTypes [#1, #2, #3]})
 ]}
