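# Material for Apollo mountaintop rocks that blends tiling cliff rock with a grass or snow top layer and into the landscape.
# Rock combines detail textures with baked normal, AO and curvature maps, height-based darkening and optional waterfall wetness masked by vertex blue and noise.
# The top layer uses a noise-broken world-up mask scaled by AO and vertex red, plus grass height for grass.
# An optional runtime virtual texture blend fades by height above landscape and vertex red.
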Material /Game/Environments/Apollo/Rocks/MountainTop/Materials/M_Apollo_MountainTop_Blend
graph {
 properties {bTangentSpaceNormal=false, bUseMaterialAttributes=true, pixelDepthOffsetMode=PDOM_Legacy, usedShadingModels=Default Lit}
 nodes [
  n1 = TextureCoordinate() {uTiling=15, vTiling=15}
  n2 = VertexColor()
  n3 = Multiply(A=TextureSample(UVs=@n1) {samplerType=SAMPLERTYPE_LinearGrayscale, texture=/Game/Global/Textures/Default/Noise/T_SwirlNoise_H2}.G, B=@n2.B)
  n4 = TextureSampleParameter2D() {
   parameterName=Baked Normal
   samplerType=SAMPLERTYPE_Normal
   texture=/Game/Environments/Apollo/Rocks/MountainTop/Textures/T_MountainTop_Waterfall_N
  }
  n5 = VectorParameter() {parameterName=Rock Tiling, defaultValue=<rgba 5 5 1 0>}
  n6 = Add(A=Multiply(A=TextureCoordinate(), B=ComponentMask(None=@n5.RGB) {g=true, r=true}), B=Constant2Vector() {g=-0.05})
  n7 = TextureSample(UVs=@n6) {
   samplerSource=SSM_Wrap_WorldGroupSettings
   samplerType=SAMPLERTYPE_Normal
   texture=/Game/Environments/Apollo/Sets/Terrain/Textures/T_Apollo_Cliff_N_03
  }
  n8 = MaterialFunctionCall(Normal=@n7.RGB, Flatness=ScalarParameter() {parameterName=Rock_NormalFlatten}) {materialFunction=&FlattenNormal}
  n9 = TextureSample(UVs=@n6) {
   samplerSource=SSM_Wrap_WorldGroupSettings
   samplerType=SAMPLERTYPE_Masks
   texture=/Game/Environments/Apollo/Sets/Terrain/Textures/T_Apollo_Cliff_S_03
  }
  n10 = ScalarParameter() {defaultValue=0.8, parameterName=Ruins_DarkenBaseColor}
  n11 = OneMinus(None=ComponentMask(None=WorldPosition().XYZ) {b=true})
  n12 = Constant() {r=10000}
  n13 = Saturate(None=Divide(A=Add(A=@n11, B=@n12), B=Constant() {r=30000}))
  n14 = Constant() {r=0.35}
  n15 = TextureSampleParameter2D() {parameterName=Baked AO, texture=/Game/Environments/Apollo/Rocks/MountainTop/Textures/T_MountainTop_Waterfall_AO}
  n16 = Multiply(
   A=Constant() {r=0.75}
   B=TextureSample(UVs=@n6) {samplerSource=SSM_Wrap_WorldGroupSettings, texture=/Game/Environments/Apollo/Sets/Terrain/Textures/T_Apollo_Cliff_D_03}.RGB
  )
  n17 = Multiply(
   A=Saturate(None=Add(A=@n13, B=@n14))
   B=Add(
    A=TextureSampleParameter2D() {parameterName=Baked Curve, texture=/Game/Environments/Apollo/Rocks/MountainTop/Textures/T_MountainTop_Waterfall_Curve}.RGB
    B=Multiply(A=@n15.RGB, B=@n16)
   )
  )
  n18 = SetMaterialAttributes!(
   MaterialAttributes
   Normal=Transform(None=MaterialFunctionCall(BaseNormal=@n4.RGB, AdditionalNormal=@n8.Result) {materialFunction=&BlendAngleCorrectedNormals}.Result)
   Roughness=Multiply(A=ScalarParameter() {defaultValue=0.8, parameterName=Rock_RoughMod}, B=@n9.B)
   Specular=Multiply(A=Constant() {r=0.3}, B=@n9.R)
   Base Color=StaticSwitchParameter(True=Multiply(A=@n10, B=@n17), False=@n17) {parameterName=UseRuinsDarken}
  ) {attributeSetTypes [
    #1
    #2
    #3
    #4
   ]}
  n19 = MaterialFunctionCall() {materialFunction=/Game/Athena/Apollo/Environments/Landscape/Materials/Biome/UniqueBiomeMaterialFunctions/MF_Biome_SwampGrass}
  n20 = Multiply(
   A=VectorParameter() {group=Grass, parameterName=Grass Color, defaultValue=<rgba 0.7629 0.8125 0.0101 1>}.RGB
   B=TextureSample(UVs=TextureCoordinate() {uTiling=20, vTiling=20}) {texture=/Game/Environments/Apollo/Sets/Terrain/Textures/T_Apollo_Grass_D_01}.RGB
  )
  n21 = Constant3Vector() {constant=<rgba 0 0 1 0>}
  n22 = ScalarParameter() {defaultValue=1.7122, group="Up Vector Snow/Grass ", parameterName=Up Vector Sharpness}
  n23 = Multiply(
   A=LinearInterpolate(
    A=ScalarParameter() {defaultValue=0.4, group="Up Vector Snow/Grass ", parameterName=Up Vector Breakup Min}
    B=ScalarParameter() {defaultValue=0.8, group="Up Vector Snow/Grass ", parameterName=Up Vector Breakup Max}
    Alpha=TextureSample(UVs=@n1) {texture=/Game/Global/Textures/Default/Noise/T_Tiling_Noise_01}.R
   )
   B=ScalarParameter() {defaultValue=-0.4943, group="Up Vector Snow/Grass ", parameterName=Up Vector Bias}
  )
  n24 = StaticBool() {value=true}
  n25 = Multiply(
   A=@n2.R
   B=Multiply(
    A=@n15.R
    B=MaterialFunctionCall!(In Explicit Normal, In World Vector=@n21, Blend Sharpness=@n22, Blend Bias=@n23, Clamped?=@n24, Alpha=@n2) {materialFunction=/Engine/Functions/Engine_MaterialFunctions01/AlphaBlend/WorldAlignedBlend}.w/ Explicit Normal
   )
  )
  n26 = Clamp(None=Power(Base=@n19.Height, Exp=Constant()))
  n27 = MaterialFunctionCall!(
   Base Material=StaticSwitchParameter(
    True=MaterialFunctionCall!(Mask=@n3, In=@n18) {materialFunction=&/MF_Waterfall_Wetness}.Result
    False=@n18
   ) {parameterName=Use Wetness}
   Top Material=StaticSwitchParameter(
    True=MaterialFunctionCall() {materialFunction=/Game/Athena/Apollo/Environments/Landscape/Materials/Biome/UniqueBiomeMaterialFunctions/MF_Biome_Snow}.Material
    False=SetMaterialAttributes!(MaterialAttributes=@n19.Material, Base Color=@n20) {attributeSetTypes [#4]}
   ) {parameterName=Snow?}
   Alpha=StaticSwitchParameter(True=@n25, False=Multiply(A=@n26, B=@n25)) {parameterName=Snow?}
  ) {materialFunction=/Engine/Functions/MaterialLayerFunctions/MatLayerBlend_Standard}
  n28 = RuntimeVirtualTextureSample!(UVs, World Position) {materialType=BaseColor_Normal_Specular_YCoCg, virtualTexture=/Game/Athena/Apollo/Maps/Landscape/VT/RVT_Apollo_Landscape}
  n29 = SetMaterialAttributes!(MaterialAttributes, Base Color=@n28.BaseColor, Specular=@n28.BaseColor, Roughness=@n28.BaseColor, Normal=@n28.BaseColor) {attributeSetTypes [
    #4
    #3
    #2
    #1
   ]}
  n30 = WorldPosition() {worldPositionShaderOffset=WPT_ExcludeAllShaderOffsets}
  n31 = Subtract(
   A=Add(A=ComponentMask(None=@n30.XYZ) {b=true}, B=ScalarParameter() {defaultValue=-8000, group=RVT, parameterName=RVT WP Offset})
   B=RuntimeVirtualTextureSample() {materialType=WorldHeight, virtualTexture=/Game/Athena/Apollo/Maps/Landscape/VT/RVT_Apollo_Landscape_Height}.WorldHeight
  )
  n32 = ScalarParameter() {defaultValue=200, group=RVT, parameterName=Ground fade}
  n33 = VertexColor()
  n34 = LinearInterpolate(B=Saturate(None=Divide(A=@n31, B=@n32)), Alpha=Max(A=@n33.R, B=@n33.R)) {constA=1}
  n35 = StaticSwitchParameter(
   True=MaterialFunctionCall(NoRVT=@n27.Blended Material, RVT=BlendMaterialAttributes(A=@n29, B=@n27.Blended Material, Alpha=@n34)) {materialFunction=&FN_RVT_Switch_Material}.Result
   False=@n27.Blended Material
  ) {defaultValue=true, parameterName=UseVirtualTexture}
  ShadingPathSwitch(Default=@n35, Deferred=@n35)
 ]
}
